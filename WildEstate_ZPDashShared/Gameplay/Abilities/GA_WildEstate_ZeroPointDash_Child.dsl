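; Child gameplay ability for the Zero Point Dash foraged item that caches the avatar player pawn on activation.
; CanActivateAbility fails with a cooldown tag while the dash cooldown tag is present.
; Commit logic checks a down-but-not-out tag query, a scalable-float allowance and a required-tag query against combined player and weapon tags.
; Optional fall-damage immunity is applied as an independent effect so it survives ability cancellation.
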(event Ability|EventOnEndAbility
  (params (bWasCancelled :type Boolean))
  (|CallParentFunction bWasCancelled))

(event Ability|EventActivateAbility
  (bind _asfort_player_pawn (Utilities|Casting|CastToFortPlayerPawn (Ability|GetAvatarActorfromActorInfo)))
  (Variables|Default|SetPlayer%20Pawn _asfort_player_pawn)
  (|CallParentFunction))

(event Ability|EventCommitExecute
  (|CallParentFunction))

(fn AttemptCommitAbility
  (if (GameplayTags|DoesContainerMatchTagQuery (GameplayTags|GetOwnedGameplayTags (Variables|Default|GetPlayer%20Pawn) :!class /Script/GameplayTags.BlueprintGameplayTagLibrary) (Variables|Default|GetTag%20Query%20DBNO))
    (if (ScalableFloat|GetValueasBool (Variables|Default|GetHF_AllowWhileDBNO))
      (if (GameplayTags|DoesContainerMatchTagQuery (Default|CombinePlayerandWeaponTags) (Variables|Default|GetRequiredTagExpressionQuery))
        (if true
          (bind _returnvalue (Ability|CommitAbility))
          (return _returnvalue)
          (else
            _))
        (else
          _))
      (return false)
      (else
        _))))

(fn CombinePlayerAndWeaponTags
  (bind _player_pawn (Variables|Default|GetPlayer%20Pawn))
  (bind _returnvalue (Weapon|GetCurrentWeapon _player_pawn))
  ;[0.4532 0.6321 1] CombinePlayerAndWeaponTags
  ; Get and combine the tags on the owning player and the current weapon
  (Utilities|FlowControl|Sequence
    (:then_0
      (LocalVariables|Seth_CombinedTagContainer (LocalVariables|Geth_CombinedTagContainer))
      (Utilities|IsValid _returnvalue
        (:Is%20Valid
          (GameplayTags|AppendGameplayTagContainers (LocalVariables|Geth_CombinedTagContainer) (Tags|GetWeaponDataGameplayTags _returnvalue)))))
    (:then_1
      ;;[0.977 0.5202 1] RETURN: combined tags from player and current weapon (if possible)
      (return (LocalVariables|Geth_CombinedTagContainer)))) ; sequence[0] get and combine tags\nsequence[1] return combined tags
)

(fn GiveTillFallDamageImmunity
  (if (ScalableFloat|GetValueasBool (Variables|Default|GetHF_FallDamageImmunity))
    (bind _returnvalue (Ability|ApplyGameplayEffectToOwner (Variables|Default|GetGE%20Fall%20Damage%20Immunity) 0)) ; needs to be applied independently so it can persist \neven if other abilties are cancelled
    (Variables|Default|SetGEH%20Fall%20Damage%20Immunity _returnvalue)))

(fn GiveTillFallDamageImmunity_0
  (if (ScalableFloat|GetValueasBool (Variables|Default|GetHF_FallDamageImmunity))
    (bind _returnvalue (Ability|ApplyGameplayEffectToOwner (Variables|Default|GetGE%20Fall%20Damage%20Immunity) 0)) ; needs to be applied independently so it can persist \neven if other abilties are cancelled
    (Variables|Default|SetGEH%20Fall%20Damage%20Immunity _returnvalue)))

(fn CanActivateAbility
  (params (ActorInfo :type Gameplay%20Ability%20Actor%20Info%20Structure) (Handle :type Gameplay%20Ability%20Spec%20Handle%20Structure))
  (bind (_owneractor _avataractor _playercontroller _abilitysystemcomponent) (Utilities|Struct|BreakGameplayAbilityActorInfo ActorInfo))
  (if (GameplayTags|HasMatchingGameplayTag _abilitysystemcomponent "(TagName=\"Granted.Athena.ForagedItem.Cooldown.ZeroPoint.Dash\")")
    (return "(GameplayTags=((TagName=\"Athena.Ability.Failed.Cooldown\")))" false)
    (else
      (return 0 true))))
